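# Sound cue named for Sneaky Snowman gadget footsteps, with no sound nodes or root in its graph.
# It carries only playback settings: third-person walk footstep attenuation, the enemy third-person foley sound class and reduced volume.
# It references a local footstep concurrency set and also holds override values for one voice per owner, stopping the oldest.
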SoundCue /FNGameplayCues/Sounds/Fort_Gadgets/SneakySnowman/SneakySnowman_Footstep_Cue
root=null
nodes []
properties {
 attenuationSettings=/Game/Sounds/Attenuation/Fort_Player_Footsteps_3P_Walk_Athena_Att
 maxDistance=3584
 soundClassObject=&Fort_Player_Foley_3P_Enemy_Athena
 volumeMultiplier=0.65
 concurrencyOverrides {bLimitToOwner=true, maxCount=1, resolutionRule=StopOldest}
 concurrencySet [&Athena_Footstep_Local_Conc]
}
